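# First-person fire sound cue for the PastaRipper weapon.
# A weighted random pick of three close ADS fire waves forms the main layer, favoring the second and third.
# The first two quality-level branches mix in a quieter shotgun sub punch, while the last two play only the random shot.
# A root modulator slightly varies pitch and volume, and the cue uses first-person weapon attenuation, sound class and gunfire concurrency.
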SoundCue /MotherGameplay/Sounds/PastaRipper/PastaRipper_Fire1P_Cue
root=Modulator_0
nodes [
 {id=Mixer_1, children [Random_3, WavePlayer_12], properties {inputVolume [1,0.2]}}
 {id=Modulator_0, children [QualityLevel_0], properties {pitchMax=1, pitchMin=0.9, volumeMax=1, volumeMin=0.9}}
 {id=QualityLevel_0, children [Mixer_1,Mixer_1,Random_3,Random_3]}
 {id=Random_3, children [WavePlayer_4, WavePlayer_10, WavePlayer_11], properties {weights [1,2,2]}}
 {id=WavePlayer_10, properties {soundWaveAssetPtr=&/PastaRipper_Fire_ADS_Close_02}}
 {id=WavePlayer_11, properties {soundWaveAssetPtr=&/PastaRipper_Fire_ADS_Close_03}}
 {id=WavePlayer_12, properties {soundWaveAssetPtr=/Game/Sounds/Fort_Weapons/Shotgun/Charge/Sub_Punch_04}}
 {id=WavePlayer_4, properties {soundWaveAssetPtr=&/PastaRipper_Fire_ADS_Close_01}}
]
properties {
 attenuationSettings=&Fort_Weapon_Projectile_FirstPerson_Athena_Att
 maxDistance=2097152
 soundClassObject=&Fort_WeaponRanged_Fire_1P_PSM
 volumeMultiplier=0.5
 attenuationOverrides {bAttenuate=false, stereoSpread=5000}
 concurrencyOverrides {bLimitToOwner=true, maxCount=4, resolutionRule=StopOldest}
 concurrencySet [&Weapon_Gunfire_FirstPerson_Conc]
}
