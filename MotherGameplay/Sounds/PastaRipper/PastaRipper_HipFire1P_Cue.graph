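# Sound cue for the PastaRipper weapon's first-person hip-fire shot, played without distance attenuation in the first-person ranged-fire sound class.
# It picks one of three close fire waves at equal weight, with slight random pitch and volume modulation.
# A quality-level node gives the first two levels that pick mixed with a quiet shotgun sub-punch layer, and the last two the pick alone.
# First-person gunfire concurrency is assigned.
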SoundCue /MotherGameplay/Sounds/PastaRipper/PastaRipper_HipFire1P_Cue
root=Modulator_2
nodes [
 {id=Mixer_0, children [Random_2,WavePlayer_3], properties {inputVolume [1,0.07]}}
 {id=Modulator_2, children [QualityLevel_1], properties {pitchMax=1, pitchMin=0.9, volumeMax=1, volumeMin=0.9}}
 {id=QualityLevel_1, children [Mixer_0,Mixer_0,Random_2,Random_2]}
 {id=Random_2, children [WavePlayer_0,WavePlayer_1,WavePlayer_2], properties {weights [1,1,1]}}
 {id=WavePlayer_0, properties {soundWaveAssetPtr=&/PastaRipper_Fire_Hip_Close_01}}
 {id=WavePlayer_1, properties {soundWaveAssetPtr=&/PastaRipper_Fire_Hip_Close_02}}
 {id=WavePlayer_2, properties {soundWaveAssetPtr=&/PastaRipper_Fire_Hip_Close_03}}
 {id=WavePlayer_3, properties {soundWaveAssetPtr=/Game/Sounds/Fort_Weapons/Shotgun/Charge/Sub_Punch_04}}
]
properties {
 attenuationSettings=&Fort_Weapon_Projectile_FirstPerson_Athena_Att
 maxDistance=2097152
 pitchMultiplier=1.15
 soundClassObject=&Fort_WeaponRanged_Fire_1P_PSM
 volumeMultiplier=0.5
 attenuationOverrides {bAttenuate=false, stereoSpread=5000}
 concurrencyOverrides {bLimitToOwner=true, maxCount=4, resolutionRule=StopOldest}
 concurrencySet [&Weapon_Gunfire_FirstPerson_Conc]
}
