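# Sound cue for the PastaRipper weapon's hip-fire gunshot, layering close, medium and distant recordings under global and local gunfire concurrency sets.
# Each layer picks randomly from three waves, then passes through a modulator, its own distance attenuation and a sound class override into one mixer.
# Only the close layer varies pitch and volume; medium and distant layers use progressively lower fixed volumes and share the distant-fire sound class.
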SoundCue /MotherGameplay/Sounds/PastaRipper/PastaRipper_HipFire_Cue
root=Mixer_0
nodes [
 {id=Attenuation_2, children [Modulator_2], properties {attenuationSettings=&Fort_Weapon_Projectile_Close_Athena_Att}}
 {
  id=Attenuation_3
  children [Modulator_3]
  properties {attenuationSettings=&Fort_Weapon_Projectile_Medium_Mono_Athena_Att}
 }
 {
  id=Attenuation_4
  children [Modulator_4]
  properties {attenuationSettings=&Fort_Weapon_Projectile_Distant_Mono_Athena_Att}
 }
 {id=Mixer_0, children [SoundClass_2,SoundClass_4,SoundClass_3], properties {inputVolume [1,0.8,0.5]}}
 {id=Modulator_2, children [Random_2], properties {pitchMax=1.1, pitchMin=1, volumeMax=1, volumeMin=0.9}}
 {id=Modulator_3, children [Random_0], properties {volumeMax=0.7, volumeMin=0.7}}
 {id=Modulator_4, children [Random_1], properties {volumeMax=0.3, volumeMin=0.3}}
 {id=Random_0, children [WavePlayer_12, WavePlayer_13, WavePlayer_14], properties {weights [1,1,1]}}
 {id=Random_1, children [WavePlayer_15, WavePlayer_16, WavePlayer_17], properties {weights [1,1,1]}}
 {id=Random_2, children [WavePlayer_4, WavePlayer_10, WavePlayer_11], properties {weights [1,1,1]}}
 {
  id=SoundClass_2
  children [Attenuation_2]
  properties {soundClassOverride=&Fort_WeaponRanged_Fire_3P_PSM}
 }
 {
  id=SoundClass_3
  children [Attenuation_4]
  properties {soundClassOverride=&Fort_WeaponRanged_Fire_Distant}
 }
 {
  id=SoundClass_4
  children [Attenuation_3]
  properties {soundClassOverride=&Fort_WeaponRanged_Fire_Distant}
 }
 {id=WavePlayer_10, properties {soundWaveAssetPtr=&/PastaRipper_Fire_Hip_Close_02}}
 {id=WavePlayer_11, properties {soundWaveAssetPtr=&/PastaRipper_Fire_Hip_Close_03}}
 {id=WavePlayer_12, properties {soundWaveAssetPtr=&/PastaRipper_Fire_Hip_Med_01}}
 {id=WavePlayer_13, properties {soundWaveAssetPtr=&/PastaRipper_Fire_Hip_Med_02}}
 {id=WavePlayer_14, properties {soundWaveAssetPtr=&/PastaRipper_Fire_Hip_Med_03}}
 {id=WavePlayer_15, properties {soundWaveAssetPtr=&/PastaRipper_Fire_Hip_Far_01}}
 {id=WavePlayer_16, properties {soundWaveAssetPtr=&/PastaRipper_Fire_Hip_Far_02}}
 {id=WavePlayer_17, properties {soundWaveAssetPtr=&/PastaRipper_Fire_Hip_Far_03}}
 {id=WavePlayer_4, properties {soundWaveAssetPtr=&/PastaRipper_Fire_Hip_Close_01}}
]
properties {
 maxDistance=27500
 pitchMultiplier=1.15
 volumeMultiplier=0.6
 attenuationOverrides {bAttenuate=false, stereoSpread=5000}
 concurrencyOverrides {maxCount=4}
 concurrencySet [&Weapon_Gunfire_Global_Conc, &Weapon_Gunfire_Local_Conc]
}
