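# Sound cue for the PastaRipper weapon's low-ammo fire sound.
# It mixes two low-ammo fire waves, the second boosted, with one of three randomly chosen impact sizzle waves.
# A modulator varies pitch and volume, and a short delay precedes playback.
# Attenuation is disabled, a low-ammo ranged-weapon sound class is assigned, and a concurrency override prevents new instances once the limit is reached.
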SoundCue /MotherGameplay/Sounds/PastaRipper/PastaRipper_LowAmmo_Cue
root=Delay_1
nodes [
 {id=Delay_1, children [Modulator_1], properties {delayMax=0.08, delayMin=0.07}}
 {id=Mixer_0, children [Random_0,WavePlayer_3,WavePlayer_5], properties {inputVolume [1,1,1.4]}}
 {id=Modulator_1, children [Mixer_0], properties {pitchMax=1.3, pitchMin=1, volumeMax=1.8, volumeMin=1.7}}
 {id=Random_0, children [WavePlayer_0,WavePlayer_1,WavePlayer_2], properties {weights [1,1,1]}}
 {id=WavePlayer_0, properties {soundWaveAssetPtr=&/PastaRipper_Impact_Sizzle_02}}
 {id=WavePlayer_1, properties {soundWaveAssetPtr=&/PastaRipper_Impact_Sizzle_03}}
 {id=WavePlayer_2, properties {soundWaveAssetPtr=&/PastaRipper_Impact_Sizzle_04}}
 {id=WavePlayer_3, properties {soundWaveAssetPtr=&/PastaRipper_Fire_LowAmmo}}
 {id=WavePlayer_5, properties {soundWaveAssetPtr=&/PastaRipper_Fire_LowAmmo_v2}}
]
properties {
 bOverrideConcurrency=true
 maxDistance=2097152
 soundClassObject=/Game/Sounds/FortSoundClassesAndModes/Fort_WeaponRanged_LowAmmo_Degraded_PSM
 volumeMultiplier=0.8
 attenuationOverrides {bAttenuate=false, stereoSpread=5000}
 concurrencyOverrides {resolutionRule=PreventNew}
}
