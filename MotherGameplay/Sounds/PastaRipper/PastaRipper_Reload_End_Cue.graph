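# Sound cue for the end of the PastaRipper weapon's reload, classed as weapon foley.
# A Modulator applies slight random pitch and volume variation to a single wave player using the EndnSwap reload sound.
# The cue plays at reduced volume, uses the shared Fort weapon reload attenuation asset, and overrides concurrency to cap simultaneous instances.
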SoundCue /MotherGameplay/Sounds/PastaRipper/PastaRipper_Reload_End_Cue
root=Modulator_0
nodes [
 {id=Modulator_0, children [WavePlayer_1], properties {pitchMax=1.1, pitchMin=0.9, volumeMax=1.1, volumeMin=0.9}}
 {id=WavePlayer_1, properties {soundWaveAssetPtr=&/PastaRipper_Reload_EndnSwap_01}}
]
properties {
 attenuationSettings=&Fort_Weapon_Reload_Att
 bOverrideConcurrency=true
 maxDistance=2250
 soundClassObject=&Fort_WeaponFoley
 volumeMultiplier=0.251
 attenuationOverrides {bAttenuate=false, stereoSpread=5000}
 concurrencyOverrides {maxCount=6}
}
