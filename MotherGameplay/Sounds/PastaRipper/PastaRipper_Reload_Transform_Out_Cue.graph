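# Sound cue for the PastaRipper weapon's reload transform-out foley, reusing an ADS-out wave as its only sample.
# The single wave passes through a one-entry random node, a SoundClass override to weapon foley, and a fixed volume-boost modulator into the weapon reload attenuation.
# The cue also lowers overall volume with its multiplier and overrides concurrency to cap simultaneous instances.
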SoundCue /MotherGameplay/Sounds/PastaRipper/PastaRipper_Reload_Transform_Out_Cue
root=Attenuation_0
nodes [
 {id=Attenuation_0, children [Modulator_1], properties {attenuationSettings=&Fort_Weapon_Reload_Att}}
 {id=Modulator_1, children [SoundClass_0], properties {pitchMax=1, pitchMin=1, volumeMax=1.8, volumeMin=1.8}}
 {id=Random_0, children [WavePlayer_0], properties {weights [1]}}
 {id=SoundClass_0, children [Random_0], properties {soundClassOverride=&Fort_WeaponFoley}}
 {id=WavePlayer_0, properties {soundWaveAssetPtr=&/PastaRipper_ADSOut_01_M}}
]
properties {
 attenuationSettings=&Fort_Weapon_Reload_Att
 bOverrideConcurrency=true
 maxDistance=2250
 soundClassObject=&Fort_WeaponFoley
 volumeMultiplier=0.301
 attenuationOverrides {bAttenuate=false, stereoSpread=5000}
 concurrencyOverrides {maxCount=6}
}
